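; Glow Donut sludge field actor that spawns a timed stink cloud hazard and triggers its damage ability through a gameplay event after an initial delay.
; On BeginPlay, audio and post-process blend fade in over that delay, and a timer covers delay plus cloud duration.
; The timer starts fade-out, deactivates the Niagara cloud and sets a very short lifespan.
; The shared fade timeline's Finished pin registers or unregisters an AI danger zone by fade direction.
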(event EventBeginPlay
  ;[1 1 1] Begin fading VFX and audio in and set timer to kill the actor after the specified duration. Since there is a delay before the damage begins and we want the duration to be the actual time that damage is being dealt, we want to kill the actor after the duration + damage delay
  (CallFunction|FadeIn)
  (Utilities|Time|SetTimerbyEvent (EventDispatchers|CreateEvent :!event Custom|HideAndKill) (+ (ScalableFloat|GetValueatLevel (Variables|Default|GetHF_CloudDuration)) (ScalableFloat|GetScalableFloatMultiplier (Variables|Default|GetHF_InitialDelay))))
  ;[1 1 1] Send event to trigger ability after delay
  (Utilities|FlowControl|Delay (ScalableFloat|GetScalableFloatMultiplier (Variables|Default|GetHF_InitialDelay)))
  (Ability|SendGameplayEventtoActor self (Variables|Default|GetSpawnCloudGA_EventTag)))

(event Custom|HideAndKill
  ;[1 1 1] Trigger VFX and audio fade out and prepare actor to be destroyed
  (CallFunction|FadeOut)
  (Actor|SetLifeSpan 0.1))

(event Custom|FadeIn
  (bind (_origin _boxextent _sphereradius) (Collision|GetComponentBounds (Variables|Default|GetDangerSphere)))
  (bind _returnvalue (Class|AthenaAIServiceZoneManager|GetAthenaAIServiceZoneManager true))
  (bind _hf_initialdelay (Variables|Default|GetHF_InitialDelay))
  (bind _returnvalue_2 (ScalableFloat|GetScalableFloatMultiplier _hf_initialdelay))
  (bind _self self)
  ;[1 1 1] Fade in and out for VFX, audio, and post processing
  (Variables|Default|SetIsFadingIn true)
  (Components|Timeline|SetPlayRate (Variables|BGA_WildEstate_GlowDonut_Sludge_Field_Athena|GetFadeSmokeTL) (Utilities|Reciprocal _returnvalue_2))
  (Audio|Components|Audio|FadeIn (Variables|Default|GetLoopingAudio) _returnvalue_2 1.0 0.0 "Linear" :!class /Script/Engine.AudioComponent)
  (bind (_direction _float) (Timeline|FadeSmokeTL Play)
    (:Update
      (Variables|/Script/Engine.PostProcessComponent|SetBlendWeight (* _float 0.7) (Variables|Default|GetPostProcessComponent)))
    (:Finished
      (Utilities|IsValid _returnvalue
        (:Is%20Valid
          (if (Variables|Default|GetIsFadingIn)
            (Class|AthenaAIServiceZoneManager|RegisterDangerZone _returnvalue _self _sphereradius)
            (else
              (Class|AthenaAIServiceZoneManager|UnregisterDangerZone _returnvalue _self))))))))

(event Custom|FadeOut
  (bind (_origin _boxextent _sphereradius) (Collision|GetComponentBounds (Variables|Default|GetDangerSphere)))
  (bind _returnvalue (Class|AthenaAIServiceZoneManager|GetAthenaAIServiceZoneManager true))
  (bind _self self)
  ;[1 1 1] Fade in and out for VFX, audio, and post processing
  (Variables|Default|SetIsFadingIn false)
  (Components|Timeline|SetPlayRate (Variables|BGA_WildEstate_GlowDonut_Sludge_Field_Athena|GetFadeSmokeTL) (Utilities|Reciprocal (Variables|Default|GetGasFadeOutTime)))
  (Audio|Components|Audio|FadeOut (Variables|Default|GetLoopingAudio) (Variables|Default|GetGasFadeOutTime) 0.0 "Linear" :!class /Script/Engine.AudioComponent)
  (Components|Activation|Deactivate (Variables|Default|GetNS_StinkCloud))
  (Timeline|FadeSmokeTL ReverseFromEnd))

(fn ConstructionScript
  (|CallParentFunction))
