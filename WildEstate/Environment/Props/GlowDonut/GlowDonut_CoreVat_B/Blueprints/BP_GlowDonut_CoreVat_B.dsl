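; Destructible GlowDonut Core Vat prop that explodes on death, affecting nearby pawns and leaving a sludge hazard.
; On server death, gated by a hotfix bool and death-effect tags, it sphere-overlaps a scale-adjusted radius and applies a gameplay effect to pawns.
; It then spawns a sludge field gameplay actor above itself and calls AddExplosionDecal for a fading impact decal and Niagara FX.
; ShouldDie calls its parent but always returns true.
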(event EventBeginPlay)

(event Collision|EventActorBeginOverlap
  (params (OtherActor :type Actor%20Object%20Reference)))

(event EventTick
  (params (DeltaSeconds :type Float%20%28single-precision%29)))

(event EventOnDeathServer
  (params (Damage :type Float%20%28single-precision%29) (DamageTags :type Gameplay%20Tag%20Container%20Structure%20%28by%20ref%29) (Momentum :type Vector :default "0, 0, 0") (HitInfo :type Hit%20Result%20Structure%20%28by%20ref%29) (InstigatedBy :type Controller%20Object%20Reference) (DamageCauser :type Actor%20Object%20Reference) (EffectContext :type Gameplay%20Effect%20Context%20Handle%20Structure))
  (bind (_returnvalue_location _returnvalue_rotation _returnvalue_scale) (Math|Transform|BreakTransform (Transformation|GetActorTransform)))
  (bind (_returnvalue_location_x _returnvalue_location_y _returnvalue_location_z) (Math|Vector|BreakVector _returnvalue_location))
  ;[1 0.4431 0] EventOnDeathServer
  ; Explode and grant NeonBurn effect to pawns in the radius, attempt to spawn a NeonBurn related pick up (if applicable)
  (if (ScalableFloat|GetValueasBool (Variables|Hotfixes|GetHF_AllowGameplayExplosion))
    (if (Class|AthenaFunctionLibrary|ShouldAllowDeathEffects DamageTags)
      (Variables|Explosion|SetExplosionInstigator InstigatedBy)
      (Utilities|FlowControl|Sequence
        (:then_0
          (bind (_outactors _returnvalue) (Collision|SphereOverlapActors (Transformation|GetActorLocation) (* (Variables|Explosion|GetExplosion%20Radius) (Math|Vector|VectorLength (Transformation|GetActorScale3D))) (Variables|Explosion|GetExplosionHitObjectTypes)))
          (if _returnvalue
            (for _array_element _outactors
              ;;[1 0.1581 0.3333] For Each Pawn or environment actor hit in the Splash radius, deal explosive damage
              (Utilities|IsValid _array_element
                (:Is%20Valid
                  (bind _output_get (Variables|Explosion|SetExplosionHitActor _array_element))
                  ;;;[0 0.06265 1] Attempt to heal any pawn hit by the splash effect
                  (bind _asfort_pawn (Utilities|Casting|CastToFortPawn _output_get))
                  (GameplayEffects|FortApplyGameplayEffectToTarget (Ability|GetFortAbilitySystemComponent self) (Variables|Default|GetGameplayEffect) (Ability|GetAbilitySystemComponent _asfort_pawn) self self 0 0.0 (select (Utilities|IsValid (Ability|GetAbilitySystemComponent (Variables|Explosion|GetExplosionInstigator)) :!class /Script/Engine.KismetSystemLibrary) (Class|AthenaFunctionLibrary|MakeSimpleFortEffectContext (Variables|Explosion|GetExplosionInstigator)) 0)))))))
        (:then_1
          ;[1 1 1] Spawn Stink Cloud BGA
          (Building|SpawnBuildingGameplayActor "/WildEstate/Environment/Props/GlowDonut/GlowDonut_CoreVat_B/BGA_WildEstate_GlowDonut_Sludge_Field_Athena.BGA_WildEstate_GlowDonut_Sludge_Field_Athena_C" (Math|Transform|MakeTransform (Math|Vector|MakeVector _returnvalue_location_x _returnvalue_location_y (+ _returnvalue_location_z 1.5e+02)) _returnvalue_rotation _returnvalue_scale) (Variables|Explosion|GetExplosionInstigator) false false)
          (CallFunction|AddExplosionDecal))))))

(event Custom|AddExplosionDecal
  (bind _returnvalue (Transformation|GetActorLocation))
  ;[1 1 1] AddExplosionDecal
  ; -Multicast event to add the decal on explosion
  ; -Spawn Death FX here to show up on mobile
  (bind _returnvalue_2 (Rendering|Decal|SpawnDecalatLocation "/WildEstate/Environment/Props/GlowDonut/GlowDonut_CoreVat_B/MD_WildEstate_GlowDonut_Decal_Impact.MD_WildEstate_GlowDonut_Decal_Impact" (Variables|Decal|GetDecal%20Size) _returnvalue (Variables|Decal|GetDecal%20Rotation) (ScalableFloat|GetValueatLevel (Variables|Hotfixes|GetHF_DecalLifeSpan))))
  (Rendering|Components|Decal|SetFadeOut _returnvalue_2 0.1 (ScalableFloat|GetValueatLevel (Variables|Hotfixes|GetHF_DecalFadeOutDuration)) false)
  (Niagara|SpawnSystematLocation (Variables|Default|GetExplosion%20FX) _returnvalue))

(fn ShouldDie
  (params (Damage :type Float%20%28single-precision%29) (EventInstigator :type Controller%20Object%20Reference) (DamageCauser :type Actor%20Object%20Reference) (bIsTest :type Boolean) (EffectContext :type Gameplay%20Effect%20Context%20Handle%20Structure%20%28by%20ref%29))
  (|CallParentFunction Damage EventInstigator DamageCauser bIsTest EffectContext)
  (return true))

(fn ConstructionScript
  (|CallParentFunction))
